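# Material function for the ShadeToll backpack's animated heart monitor, layering a BPM readout and sweeping scan lines onto the incoming emissive color.
# A sine-varied BPM value is split into tens and ones, driving one FlipBook digit atlas lookup with offset UVs per digit.
# Masked, time-panned lines get two-tone gradient colors, with HeartRateState and HighBMPBool scalars exposed as state inputs.
# A separate output adjusts temporal responsiveness from the TempResp input.
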MaterialFunction /BRCosmetics/Accessories/FORT_Backpacks/Backpack_F_MED_ShadeToll/Material/MF_HeartMonitorShadeToll
graph {nodes [
  n1 = GetMaterialAttributes(None=FunctionInput() {inputName=MA, inputType=FunctionInput_MaterialAttributes}) {attributeGetTypes [#1]}
  n2 = TextureObjectParameter() {
   group=HeartMonitor
   parameterName=Background
   texture=/BRCosmetics/Accessories/FORT_Backpacks/Backpack_F_MED_ShadeToll/Textures/T_ShadeToll_Backpack_CL
  }
  n3 = LinearInterpolate(
   A=@n1.EmissiveColor
   B=TextureSample(Tex=@n2) {texture=&T_ShadeToll_Backpack_CL}.RGB
   Alpha=NamedRerouteUsage()
  ) {constB=0}
  n4 = Add(A=NamedRerouteUsage(), B=NamedRerouteUsage())
  FunctionOutput(None=SetMaterialAttributes!(MaterialAttributes=@n1.MaterialAttributes, EmissiveColor=Add(A=@n3, B=@n4)) {attributeSetTypes [#1]})
  n5 = TextureObjectParameter() {
   group=HeartMonitor
   parameterName=TechArtFX
   texture=/BRCosmetics/Accessories/FORT_Backpacks/Backpack_F_MED_ShadeToll/Textures/T_ShadeToll_Backpack_FX
  }
  NamedRerouteDeclaration(None=TextureSample(Tex=@n5).G) {name=FX Mask Green, nodeColor=<rgba 0 1 0 0>}
  n6 = TextureObjectParameter() {
   group=HeartMonitor
   parameterName=TechArtFX1
   texture=/BRCosmetics/Accessories/FORT_Backpacks/Backpack_F_MED_ShadeToll/Textures/T_ShadeToll_Backpack_FX1
  }
  n7 = TextureSample(Tex=@n6) {texture=&T_ShadeToll_Backpack_FX1}
  NamedRerouteDeclaration(None=@n7.G) {name=FX 1 Mask Green, nodeColor=<rgba 0 0.142 0.0371 1>}
  n8 = Multiply(A=Time(), B=ScalarParameter() {defaultValue=0.1, group=BPM, parameterName="BPM  Speed"})
  n9 = Frac(None=Multiply(A=Sine(None=@n8)))
  n10 = Saturate(None=Subtract(A=Multiply(A=@n9) {constB=2}))
  n11 = NamedRerouteUsage()
  NamedRerouteDeclaration(None=Add(A=LinearInterpolate(Alpha=@n10) {constA=53, constB=62}, B=LinearInterpolate(Alpha=@n11) {constB=35}) {constB=20}) {name=BPM, nodeColor=<rgba 1 0 0.2344 1>}
  n12 = NamedRerouteUsage()
  n13 = TextureCoordinate() {vTiling=2}
  n14 = ScalarParameter() {defaultValue=-0.5, parameterName=BPMScanLineSpeed}
  n15 = Add(A=Time(), B=LinearInterpolate(Alpha=NamedRerouteUsage()) {constB=0.25})
  n16 = Add(A=OneMinus(None=ComponentMask(None=@n13) {g=true}), B=Frac(None=Multiply(A=@n15, B=@n14) {constB=-0.3}))
  n17 = Subtract(A=@n16, B=Floor(None=@n16))
  n18 = Multiply(A=NamedRerouteUsage(), B=Saturate(None=SmoothStep(Value=@n17) {constMin=0.97}))
  n19 = ComponentMask(None=TextureCoordinate() {vTiling=2}) {g=true}
  n20 = ScalarParameter() {defaultValue=-0.5, parameterName=BPMScanLineSpeed}
  n21 = Add(A=OneMinus(None=Add(A=@n19) {constB=0.5}), B=Frac(None=Multiply(A=@n15, B=@n20) {constB=-0.3}))
  n22 = Subtract(A=@n21, B=Floor(None=@n21))
  n23 = SmoothStep(Value=@n22) {constMin=0.97}
  n24 = Multiply(A=NamedRerouteUsage(), B=Multiply(A=NamedRerouteUsage(), B=Saturate(None=@n23)))
  n25 = NamedRerouteUsage()
  NamedRerouteDeclaration(None=Add(A=Multiply(A=@n12, B=@n18), B=LinearInterpolate(B=@n24, Alpha=@n25))) {name=PanningEdge, nodeColor=<rgba 0 1 0.0848 1>}
  n26 = NamedRerouteUsage()
  n27 = LinearInterpolate(
   A=LinearInterpolate(A=Constant3Vector() {constant=<rgba 0.57 0.97 1 1>}, B=Constant3Vector() {constant=<rgba 1 0.6505 0.5583 1>}, Alpha=NamedRerouteUsage())
   B=LinearInterpolate(
    A=Constant3Vector() {constant=<rgba 1 0.8691 0.57 1>}
    B=Constant3Vector() {constant=<rgba 0.9387 0.4801 0.4653 1>}
    Alpha=NamedRerouteUsage()
   )
   Alpha=NamedRerouteUsage()
  ) {constAlpha=1}
  n28 = NamedRerouteUsage()
  n29 = LinearInterpolate(
   A=LinearInterpolate(A=Constant3Vector() {constant=<rgba 0 0.65 1 1>}, B=Constant3Vector() {constant=<rgba 0.9449 0.1726 0.1726 1>}, Alpha=NamedRerouteUsage())
   B=LinearInterpolate(A=Constant3Vector() {constant=<rgba 1 0.1739 0 1>}, B=Constant3Vector() {constant=<rgba 0.8058 0.0163 0.0163 1>}, Alpha=NamedRerouteUsage())
   Alpha=NamedRerouteUsage()
  ) {constAlpha=1}
  NamedRerouteDeclaration(None=Add(A=Multiply(A=@n26, B=@n27), B=Multiply(A=@n28, B=@n29) {constA=1})) {name=PanningLineColorFinal, nodeColor=<rgba 0 0.2969 1 1>}
  NamedRerouteDeclaration(None=ScalarParameter() {parameterName=HighBMPBool}) {name=PartMod_HighBPM, nodeColor=<rgba 1 0 0.6328 1>}
  n30 = NamedRerouteUsage()
  n31 = Add(A=NamedRerouteUsage(), B=NamedRerouteUsage())
  n32 = Constant3Vector() {constant=<rgba 0 1 0 0>}
  n33 = Constant3Vector() {constant=<rgba 1 0.2878 0 1>}
  n34 = NamedRerouteUsage()
  NamedRerouteDeclaration(None=Multiply(A=Multiply(A=@n30, B=@n31), B=LinearInterpolate(A=@n32, B=@n33, Alpha=@n34))) {name=HeartBPMColorFinal, nodeColor=<rgba 0.6172 1 0 1>}
  n35 = NamedRerouteUsage()
  n36 = Multiply(A=Floor(None=Divide(A=@n35) {constB=10})) {constB=0.1}
  n37 = NamedRerouteUsage()
  n38 = Constant() {r=10}
  n39 = Multiply(A=Floor(None=Fmod(A=@n37, B=@n38))) {constB=0.1}
  n40 = MaterialFunctionCall(
   "Animation  Phase (0-1)"=LinearInterpolate(A=Saturate(None=@n36), B=Saturate(None=@n39), Alpha=NamedRerouteUsage())
   Number of Rows=ScalarParameter() {defaultValue=2, group=BPM, parameterName=Rows}
   Number of Columns=ScalarParameter() {defaultValue=5, group=BPM, parameterName=Columns}
   Texture=NamedRerouteUsage()
   UVs=LinearInterpolate(A=NamedRerouteUsage(), B=NamedRerouteUsage(), Alpha=NamedRerouteUsage())
   "Use Mip Bias (T) Level (F)"=StaticBoolParameter() {parameterName="Use Mip Bias (T) Level (F) (SB)"}
  ) {materialFunction=&FlipBook}
  n41 = ScalarParameter() {defaultValue=0.44, group=HeartMonitor, parameterName=VerticalMask}
  n42 = ComponentMask(None=TextureCoordinate()) {g=true}
  n43 = Subtract(A=Step(Y=Add(A=@n41) {constB=-0.005}, X=@n42) {constY=0.435}, B=Step(Y=Add(A=@n41) {constB=0.005}, X=@n42) {constY=0.445})
  NamedRerouteDeclaration(None=Saturate(None=Subtract(A=@n40.Alpha, B=@n43))) {name=Digits, nodeColor=<rgba 1 0 0.7734 1>}
  n44 = TextureCoordinate() {vTiling=2}
  NamedRerouteDeclaration(None=Step(Y=ComponentMask(None=@n44) {r=true}) {constX=0.2}) {name=TwoGradientColorMask, nodeColor=<rgba 0 0.0625 1 1>}
  NamedRerouteDeclaration(None=@n7.R) {name=FX 1 Mask Red, nodeColor=<rgba 0.1768 0 0 0>}
  n45 = SmoothStep(Value=@n22) {constMin=0.8}
  n46 = Multiply(A=NamedRerouteUsage(), B=Multiply(A=NamedRerouteUsage(), B=Saturate(None=@n45)))
  n47 = NamedRerouteUsage()
  n48 = NamedRerouteUsage()
  n49 = Multiply(A=NamedRerouteUsage(), B=Saturate(None=SmoothStep(Value=@n17) {constMin=0.8}))
  NamedRerouteDeclaration(None=Add(A=LinearInterpolate(B=@n46, Alpha=@n47), B=Multiply(A=@n48, B=@n49))) {name=PanningLine, nodeColor=<rgba 1 0.6797 0 1>}
  n50 = TextureCoordinate()
  n51 = ComponentMask(None=@n50) {g=true}
  n52 = MaterialFunctionCall(
   UVs=AppendVector(A=OneMinus(None=@n51), B=ComponentMask(None=@n50) {r=true})
   Texture Scale=AppendVector(
    A=ScalarParameter() {defaultValue=0.11, group=BPM, parameterName=DigitScaleX}
    B=ScalarParameter() {defaultValue=0.08, group=BPM, parameterName=DigitScaleY}
   )
  ) {materialFunction=&ScaleUVsByCenter}
  n53 = ScalarParameter() {defaultValue=-0.45, group=BPM, parameterName=DigitsA_XOffset}
  n54 = ScalarParameter() {defaultValue=-1.2, group=BPM, parameterName=DigitsA_YOffset}
  n55 = Saturate(None=Add(A=@n52.UVs, B=AppendVector(A=@n53, B=@n54)))
  n56 = ComponentMask(None=@n55) {r=true}
  NamedRerouteDeclaration(None=AppendVector(A=SmoothStep(Value=@n56) {constMin=0.5}, B=ComponentMask(None=@n55) {g=true})) {name=UV_A, nodeColor=<rgba 0.336 0.0019 1 1>}
  n57 = ScalarParameter() {defaultValue=-0.62, group=BPM, parameterName=DigitsB_XOffset}
  n58 = ScalarParameter() {defaultValue=-1.2, group=BPM, parameterName=DigitsB_YOffset}
  n59 = Saturate(None=Add(A=@n52.UVs, B=AppendVector(A=@n57, B=@n58)))
  n60 = ComponentMask(None=@n59) {r=true}
  NamedRerouteDeclaration(None=AppendVector(A=SmoothStep(Value=@n60) {constMax=0.5}, B=ComponentMask(None=@n59) {g=true})) {name=UV_B, nodeColor=<rgba 1 0 0.2109 1>}
  NamedRerouteDeclaration(None=Clamp(None=NamedRerouteUsage()) {maxDefault=99}) {name=CurrentNumber, nodeColor=<rgba 0 0.0187 1 1>}
  n61 = TextureObject() {samplerType=SAMPLERTYPE_LinearColor, texture=/BRCosmetics/Accessories/FORT_Backpacks/Backpack_F_MED_ShadeToll/Textures/T_ShadeTollBPMAtlas}
  NamedRerouteDeclaration(None=@n61) {name=DigitTexture, nodeColor=<rgba 0 0.8377 1 1>}
  NamedRerouteDeclaration(None=ScalarParameter() {parameterName=HeartRateState}) {name=PartMod_HeartRateState, nodeColor=<rgba 0 1 0.9063 1>}
  NamedRerouteDeclaration(None=OneMinus(None=Step(Y=@n41, X=@n42) {constY=0.44})) {name=DigitsLerpMask, nodeColor=<rgba 0 1 0.6953 1>}
  TemporalResponsivenessOutput!(None=LinearInterpolate(A=FunctionInput() {inputName=TempResp, inputType=FunctionInput_Scalar}, Alpha=NamedRerouteUsage()) {constB=0.5})
 ]}
