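# Physics asset for the ShadeToll backpack that defines simulation bodies for a dangling paper piece on the pack.
# A box covers the pack, and a small base_paper capsule is locked to it in swing and twist as the anchor.
# Three dyn_paper capsules chain from that anchor, with nonzero angular limits on only one swing axis for hinge-like bending.
# Those three capsules keep headband shape names, apparently left over from another asset.
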PhysicsAsset /BRCosmetics/Accessories/FORT_Backpacks/Backpack_F_MED_ShadeToll/Meshes/F_MED_ShadeToll_Pack_Physic_FE
bodies {
 base_paper [{shapeType=Capsule, length=0.505, radius=0.505, shapeName=base_paper_capsule, rotation=<rot 0 -90 0>}]
 dyn_paper_01 [{
   shapeType=Capsule
   length=5.1408
   radius=1.5856
   shapeName=dyn_headband_l_01_capsule
   center=<1.602 0.0722 -0.0056>
   rotation=<rot -89.7992 179.985 -179.985>
  }]
 dyn_paper_02 [{
   shapeType=Capsule
   length=5.6132
   radius=1.6131
   shapeName=dyn_headband_l_04_capsule
   center=<0.916 -0.2013 -0.0032>
   rotation=<rot -89.7992 179.985 -179.985>
  }]
 dyn_paper_03 [{
   shapeType=Capsule
   length=5.8196
   radius=1.9269
   shapeName=dyn_headband_l_07_capsule
   center=<2.0864 -0.1996 -0.0073>
   rotation=<rot -89.7992 179.985 -179.985>
  }]
 pack [{shapeType=Box, extentX=22.2976, extentY=10.4521, extentZ=56.1064, shapeName=pack_box, center=<0 6.6367 -4.8817>}]
}
constraints [
 {
  bone1Name=base_paper
  bone2Name=pack
  swing1LimitDegrees=45
  swing1Motion=Locked
  swing2LimitDegrees=45
  swing2Motion=Locked
  twistLimitDegrees=45
  twistMotion=Locked
 }
 {bone1Name=dyn_paper_01, bone2Name=base_paper, swing1LimitDegrees=60, swing2LimitDegrees=0, twistLimitDegrees=0}
 {bone1Name=dyn_paper_02, bone2Name=dyn_paper_01, swing1LimitDegrees=90, swing2LimitDegrees=0, twistLimitDegrees=0}
 {bone1Name=dyn_paper_03, bone2Name=dyn_paper_02, swing1LimitDegrees=75, swing2LimitDegrees=0, twistLimitDegrees=0}
]
