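# Physics asset for the ShadeToll backpack cosmetic, set up for a dangling paper piece on the pack.
# A box body covers the pack, and a small base_paper capsule is rotationally locked to it as the anchor.
# Three dyn_paper capsules chain from that anchor, each constrained to the one before.
# Their limit angles are wide on one swing axis and very narrow on the other swing and twist, suggesting hinge-like flapping.
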PhysicsAsset /BRCosmetics/Accessories/FORT_Backpacks/Backpack_F_MED_ShadeToll/Meshes/F_MED_ShadeToll_Pack_Physics
bodies {
 base_paper [{shapeType=Capsule, length=0.505, radius=0.505, shapeName=base_paper_capsule, rotation=<rot 0 -90 0>}]
 dyn_paper_01 [{shapeType=Capsule, length=3, radius=1, shapeName=dyn_paper_01_capsule, center=<2 0 0>, rotation=<rot -90 0 0>}]
 dyn_paper_02 [{shapeType=Capsule, length=3, radius=1, shapeName=dyn_paper_02_capsule, center=<2 0 0>, rotation=<rot -90 0 0>}]
 dyn_paper_03 [{shapeType=Capsule, length=3, radius=1, shapeName=dyn_paper_03_capsule, center=<2 0 0>, rotation=<rot -90 0 0>}]
 pack [{shapeType=Box, extentX=22.2976, extentY=10.4521, extentZ=56.1064, shapeName=pack_box, center=<0 6.6367 -4.8817>}]
}
constraints [
 {
  bone1Name=base_paper
  bone2Name=pack
  swing1LimitDegrees=45
  swing1Motion=Locked
  swing2LimitDegrees=45
  swing2Motion=Locked
  twistLimitDegrees=45
  twistMotion=Locked
 }
 {bone1Name=dyn_paper_01, bone2Name=base_paper, swing1LimitDegrees=45, swing2LimitDegrees=1, twistLimitDegrees=1}
 {bone1Name=dyn_paper_02, bone2Name=dyn_paper_01, swing1LimitDegrees=45, swing2LimitDegrees=1, twistLimitDegrees=1}
 {bone1Name=dyn_paper_03, bone2Name=dyn_paper_02, swing1LimitDegrees=45, swing2LimitDegrees=1, twistLimitDegrees=1}
]
