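# Animation Blueprint for the ShadeToll backpack, adding secondary motion to its dynamic wire and paper bones.
# The input pose gets pack and paper-base offsets, then wire AnimDynamics with two spring variants, cached for the final stage.
# Bool blends choose between RigidBody nodes using two physics assets and AnimDynamics paper chains with two stiffnesses.
# The event graph updates gravity and rigid-body settings from a parameter asset and derives gender and body-size flags on customization load.
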AnimBlueprint /BRCosmetics/Accessories/FORT_Backpacks/Backpack_F_MED_ShadeToll/Meshes/F_MED_ShadeToll_Pack_AnimBP
target_skeleton=&/F_MED_ShadeToll_Pack_Skeleton
graphs [
 AnimGraph {nodes [
   n1 = Anim Dynamics - Bone: dyn_paper_01!(ComponentPose=Local To Component(LocalPose=Use cached pose 'Anim Dynamics'()), Alpha=1, GravityOverride:Vector="0,0,-980") -> (Pose) {
    angularDampingOverride=1
    angularSpringConstant=6000
    bAngularSpring=true
    bOverrideAngularDamping=true
    bUseGravityOverride=true
    gravityScale=0.35
    simulationSpace=RootRelative
    boundBone {boneName=dyn_paper_01}
    componentAppliedLinearAccClamp=<3000 3000 3000>
    componentLinearAccScale=<1 1 1>
    componentLinearVelScale=<1 1 1>
    gravityOverride {z=-980}
    physicsBodyDefinitions [{
      boundBone {boneName=dyn_paper_01}
      boxExtents=<20 30 30>
      constraintSetup {angularLimitsMax=<0 0 35>, angularLimitsMin=<0 0 -90>, angularTarget=<1 0 0>}
      localJointOffset=<3 0 0>
     }]
    simSpaceSettings {maxAngularAcceleration=1, maxAngularVelocity=6}
   }
   n2 = Anim Dynamics - Bone: dyn_paper_02!(ComponentPose=@n1, Alpha=1, GravityOverride:Vector="0,0,-980") -> (Pose) {
    angularDampingOverride=1
    angularSpringConstant=6000
    bAngularSpring=true
    bOverrideAngularDamping=true
    bUseGravityOverride=true
    gravityScale=0.35
    simulationSpace=RootRelative
    boundBone {boneName=dyn_paper_02}
    componentAppliedLinearAccClamp=<3000 3000 3000>
    componentLinearAccScale=<1 1 1>
    componentLinearVelScale=<1 1 1>
    gravityOverride {z=-980}
    physicsBodyDefinitions [{
      boundBone {boneName=dyn_paper_02}
      boxExtents=<20 30 30>
      constraintSetup {angularLimitsMax=<0 0 35>, angularLimitsMin=<0 0 -90>, angularTarget=<1 0 0>}
      localJointOffset=<3 0 0>
     }]
    simSpaceSettings {maxAngularAcceleration=1, maxAngularVelocity=6}
   }
   n3 = Anim Dynamics - Bone: dyn_paper_03!(ComponentPose=@n2, Alpha=1, GravityOverride:Vector="0,0,-980") -> (Pose) {
    angularDampingOverride=1
    angularSpringConstant=6000
    bAngularSpring=true
    bOverrideAngularDamping=true
    bUseGravityOverride=true
    gravityScale=0.35
    simulationSpace=RootRelative
    boundBone {boneName=dyn_paper_03}
    componentAppliedLinearAccClamp=<3000 3000 3000>
    componentLinearAccScale=<1 1 1>
    componentLinearVelScale=<1 1 1>
    gravityOverride {z=-980}
    physicsBodyDefinitions [{
      boundBone {boneName=dyn_paper_03}
      boxExtents=<20 30 30>
      constraintSetup {angularLimitsMax=<0 0 35>, angularLimitsMin=<0 0 -90>, angularTarget=<1 0 0>}
      localJointOffset=<3 0 0>
     }]
    simSpaceSettings {maxAngularAcceleration=1, maxAngularVelocity=6}
   }
   n4 = Component To Local(ComponentPose=@n3)
   n5 = Anim Dynamics - Bone: dyn_paper_01(ComponentPose=Local To Component(LocalPose=Use cached pose 'Anim Dynamics'()), Alpha=1, GravityOverride="0,0,-980") {
    angularDampingOverride=1
    angularSpringConstant=9000
    bAngularSpring=true
    bOverrideAngularDamping=true
    bUseGravityOverride=true
    gravityScale=0.35
    simulationSpace=RootRelative
    boundBone {boneName=dyn_paper_01}
    componentAppliedLinearAccClamp=<3000 3000 3000>
    componentLinearAccScale=<1 1 1>
    componentLinearVelScale=<1 1 1>
    gravityOverride {z=-980}
    physicsBodyDefinitions [{
      boundBone {boneName=dyn_paper_01}
      boxExtents=<20 30 30>
      constraintSetup {angularLimitsMax=<0 0 35>, angularLimitsMin=<0 0 -90>, angularTarget=<1 0 0>}
      localJointOffset=<3 0 0>
     }]
    simSpaceSettings {maxAngularAcceleration=1, maxAngularVelocity=6}
   }
   n6 = Anim Dynamics - Bone: dyn_paper_02(ComponentPose=@n5, Alpha=1, GravityOverride="0,0,-980") {
    angularDampingOverride=1
    angularSpringConstant=9000
    bAngularSpring=true
    bOverrideAngularDamping=true
    bUseGravityOverride=true
    gravityScale=0.35
    simulationSpace=RootRelative
    boundBone {boneName=dyn_paper_02}
    componentAppliedLinearAccClamp=<3000 3000 3000>
    componentLinearAccScale=<1 1 1>
    componentLinearVelScale=<1 1 1>
    gravityOverride {z=-980}
    physicsBodyDefinitions [{
      boundBone {boneName=dyn_paper_02}
      boxExtents=<20 30 30>
      constraintSetup {angularLimitsMax=<0 0 35>, angularLimitsMin=<0 0 -90>, angularTarget=<1 0 0>}
      localJointOffset=<3 0 0>
     }]
    simSpaceSettings {maxAngularAcceleration=1, maxAngularVelocity=6}
   }
   n7 = Anim Dynamics - Bone: dyn_paper_03(ComponentPose=@n6, Alpha=1, GravityOverride="0,0,-980") {
    angularDampingOverride=1
    angularSpringConstant=9000
    bAngularSpring=true
    bOverrideAngularDamping=true
    bUseGravityOverride=true
    gravityScale=0.35
    simulationSpace=RootRelative
    boundBone {boneName=dyn_paper_03}
    componentAppliedLinearAccClamp=<3000 3000 3000>
    componentLinearAccScale=<1 1 1>
    componentLinearVelScale=<1 1 1>
    gravityOverride {z=-980}
    physicsBodyDefinitions [{
      boundBone {boneName=dyn_paper_03}
      boxExtents=<20 30 30>
      constraintSetup {angularLimitsMax=<0 0 35>, angularLimitsMin=<0 0 -90>, angularTarget=<1 0 0>}
      localJointOffset=<3 0 0>
     }]
    simSpaceSettings {maxAngularAcceleration=1, maxAngularVelocity=6}
   }
   n8 = Component To Local(ComponentPose=@n7)
   n9 = RigidBody!(
    ComponentPose=Local To Component(LocalPose=Use cached pose 'Anim Dynamics'())
    Alpha
    OverrideWorldGravity:Vector="0,0,0"
    ComponentLinearAccScale:Vector="0.3,0.3,0.3"
    ComponentLinearVelScale:Vector="0.35,0.35,0.35"
    SimSpaceSettings:Sim Space Settings Structure=(WorldAlpha=1,VelocityScaleZ=0,DampingAlpha=1,MaxLinearVelocity=500,MaxAngularVelocity=3,MaxLinearAcceleration=500,MaxAngularAcceleration=900,ExternalLinearDragV=(X=0.2,Y=0.2,Z=0),ExternalLinearVelocity=(X=0,Y=0,Z=0),ExternalAngularVelocity=(X=0,Y=0,Z=0))
   ) -> (Pose) {
    bClampLinearTranslationLimitToRefPose=true
    bOverrideWorldGravity=true
    overridePhysicsAsset=&/F_MED_ShadeToll_Pack_Physic_FE
    simulationSpace=BaseBoneSpace
    alphaBoolBlend {blendInTime=0.25, blendOption=Linear, blendOutTime=0.25}
    baseBoneRef {boneName=pack}
    componentAppliedLinearAccClamp=<1200 1200 1200>
    componentLinearAccScale=<0.3 0.3 0.3>
    componentLinearVelScale=<0.35 0.35 0.35>
    overrideWorldGravity {z=-980}
    simSpaceSettings {
     maxAngularAcceleration=900
     maxAngularVelocity=3
     maxLinearAcceleration=500
     maxLinearVelocity=500
     velocityScaleZ=0
     worldAlpha=1
     externalLinearDragV=<0.2 0.2>
    }
   }
   n10 = Component To Local(ComponentPose=@n9)
   n11 = RigidBody!(
    ComponentPose=Local To Component(LocalPose=Use cached pose 'Anim Dynamics'())
    Alpha
    OverrideWorldGravity="0,0,0"
    ComponentLinearAccScale="0.3,0.3,0.3"
    ComponentLinearVelScale="0.35,0.35,0.35"
   ) -> (Pose) {
    bClampLinearTranslationLimitToRefPose=true
    bOverrideWorldGravity=true
    overridePhysicsAsset=&/F_MED_ShadeToll_Pack_Physics
    simulationSpace=BaseBoneSpace
    alphaBoolBlend {blendInTime=0.25, blendOption=Linear, blendOutTime=0.25}
    baseBoneRef {boneName=pack}
    componentAppliedLinearAccClamp=<1200 1200 1200>
    componentLinearAccScale=<0.3 0.3 0.3>
    componentLinearVelScale=<0.35 0.35 0.35>
    overrideWorldGravity {z=-980}
    simSpaceSettings {
     maxAngularAcceleration=900
     maxAngularVelocity=3
     maxLinearAcceleration=500
     maxLinearVelocity=500
     velocityScaleZ=0
     worldAlpha=1
     externalLinearDragV=<0.2 0.2>
    }
   }
   n12 = Component To Local(ComponentPose=@n11)
   Output Pose(Result=Blend Poses by bool(
     BlendPose_0=Blend Poses by bool(BlendPose_0=@n10, BlendPose_1=@n12) {blendType=Linear}
     BlendPose_1=Blend Poses by bool(BlendPose_0=@n4, BlendPose_1=@n8) {blendType=Linear}
    ) {blendType=Linear}) {layerGroup=None}
   n13 = "Transform (Modify) Bone - Bone: pack"!(ComponentPose=Local To Component(LocalPose=Input Pose()), bAlphaBoolEnabled:bool, Translation="0,-0.25,0") -> (Pose) {
    alphaInputType=Bool
    translationMode=BMM_Additive
    translationSpace=BCS_BoneSpace
    alphaBoolBlend {blendOption=Linear}
    boneToModify {boneName=pack}
    translation {y=-0.25}
   }
   n14 = "Transform (Modify) Bone - Bone: pack"(ComponentPose=@n13, Translation="0,0.5,0") {
    alphaInputType=Bool
    translationMode=BMM_Additive
    translationSpace=BCS_BoneSpace
    alphaBoolBlend {blendOption=Linear}
    boneToModify {boneName=pack}
    translation {y=0.5}
   }
   n15 = "Transform (Modify) Bone - Bone: pack"(ComponentPose=@n14, Translation="0,-0.75,0") {
    alphaInputType=Bool
    translationMode=BMM_Additive
    translationSpace=BCS_BoneSpace
    alphaBoolBlend {blendOption=Linear}
    boneToModify {boneName=pack}
    translation {y=-0.75}
   }
   n16 = "Transform (Modify) Bone - Bone: pack"!(ComponentPose=@n15, Alpha:float, Translation="0,1,0") -> (Pose) {
    translationMode=BMM_Additive
    translationSpace=BCS_BoneSpace
    alphaBoolBlend {blendOption=Linear}
    boneToModify {boneName=pack}
    translation {y=1}
   }
   n17 = "Transform (Modify) Bone - Bone: base_paper"!(ComponentPose=@n16, bAlphaBoolEnabled:bool=True, Rotation="25,0,0") -> (Pose) {
    alphaInputType=Bool
    rotationMode=BMM_Additive
    rotationSpace=BCS_WorldSpace
    translationMode=BMM_Additive
    translationSpace=BCS_BoneSpace
    alphaBoolBlend {blendOption=Linear}
    boneToModify {boneName=base_paper}
    rotation {pitch=25}
   }
   n18 = "Transform (Modify) Bone - Bone: pack"!(ComponentPose=@n17, bAlphaBoolEnabled:bool, Translation="0,-2,0") -> (Pose) {
    alphaInputType=Bool
    rotationMode=BMM_Additive
    rotationSpace=BCS_BoneSpace
    translationMode=BMM_Additive
    translationSpace=BCS_BoneSpace
    alphaBoolBlend {blendOption=Linear}
    boneToModify {boneName=pack}
    translation {y=-2}
   }
   Save cached pose 'offset_CP'(Pose=Component To Local(ComponentPose=@n18)) {cacheName=offset_CP}
   n19 = Anim Dynamics - Bone: dyn_wire_01_l!(ComponentPose=Local To Component(LocalPose=Use cached pose 'offset_CP'()), Alpha=1) -> (Pose) {
    angularDampingOverride=1
    angularSpringConstant=20000
    bAngularSpring=true
    bLinearSpring=true
    bOverrideAngularDamping=true
    bOverrideLinearDamping=true
    bUseGravityOverride=true
    gravityScale=0.35
    linearDampingOverride=1
    linearSpringConstant=50
    simulationSpace=RootRelative
    alphaBoolBlend {blendOption=Linear}
    boundBone {boneName=dyn_wire_01_l}
    componentLinearAccScale=<1 1 1>
    componentLinearVelScale=<1 1 1>
    gravityOverride {z=-980}
    physicsBodyDefinitions [{
      boundBone {boneName=dyn_wire_01_l}
      boxExtents=<40 40 30>
      constraintSetup {angularLimitsMax=<0 20 0>, angularLimitsMin=<0 -20 -20>, angularTarget=<1 0 0>, linearAxesMax=<1 0 1>, linearAxesMin=<-0.25 -0.25 0>}
      localJointOffset=<-4 0 0>
     }]
    retargetingSettings {bFlipEasing=true, easingWeight=4}
   }
   n20 = Anim Dynamics - Bone: dyn_wire_02_l!(ComponentPose=@n19, Alpha=1) -> (Pose) {
    angularDampingOverride=1
    angularSpringConstant=20000
    bAngularSpring=true
    bLinearSpring=true
    bOverrideAngularDamping=true
    bOverrideLinearDamping=true
    bUseGravityOverride=true
    gravityScale=0.35
    linearDampingOverride=1
    linearSpringConstant=50
    simulationSpace=RootRelative
    alphaBoolBlend {blendOption=Linear}
    boundBone {boneName=dyn_wire_02_l}
    componentLinearAccScale=<1 1 1>
    componentLinearVelScale=<1 1 1>
    gravityOverride {z=-980}
    physicsBodyDefinitions [{
      boundBone {boneName=dyn_wire_02_l}
      boxExtents=<40 40 30>
      constraintSetup {angularLimitsMax=<0 20 0>, angularLimitsMin=<0 -20 -20>, angularTarget=<1 0 0>, linearAxesMax=<1 0 1>, linearAxesMin=<-0.25 -0.25 0>}
      localJointOffset=<-4 0 0>
     }]
    retargetingSettings {bFlipEasing=true, easingWeight=4}
   }
   n21 = Anim Dynamics - Bone: dyn_wire_01_r!(ComponentPose=@n20, Alpha=1) -> (Pose) {
    angularDampingOverride=1
    angularSpringConstant=20000
    bAngularSpring=true
    bLinearSpring=true
    bOverrideAngularDamping=true
    bOverrideLinearDamping=true
    bUseGravityOverride=true
    gravityScale=0.35
    linearDampingOverride=1
    linearSpringConstant=50
    simulationSpace=RootRelative
    alphaBoolBlend {blendOption=Linear}
    boundBone {boneName=dyn_wire_01_r}
    componentLinearAccScale=<1 1 1>
    componentLinearVelScale=<1 1 1>
    gravityOverride {z=-980}
    physicsBodyDefinitions [{
      boundBone {boneName=dyn_wire_01_r}
      boxExtents=<40 40 30>
      constraintSetup {angularLimitsMax=<0 20 0>, angularLimitsMin=<0 -20 -20>, angularTarget=<1 0 0>, linearAxesMax=<1 0 1>, linearAxesMin=<-0.25 -0.25 0>}
      localJointOffset=<4 0 0>
     }]
    retargetingSettings {bFlipEasing=true, easingWeight=4}
   }
   n22 = Anim Dynamics - Bone: dyn_wire_02_r!(ComponentPose=@n21, Alpha=1) -> (Pose) {
    angularDampingOverride=1
    angularSpringConstant=20000
    bAngularSpring=true
    bLinearSpring=true
    bOverrideAngularDamping=true
    bOverrideLinearDamping=true
    bUseGravityOverride=true
    gravityScale=0.35
    linearDampingOverride=1
    linearSpringConstant=50
    simulationSpace=RootRelative
    alphaBoolBlend {blendOption=Linear}
    boundBone {boneName=dyn_wire_02_r}
    componentLinearAccScale=<1 1 1>
    componentLinearVelScale=<1 1 1>
    gravityOverride {z=-980}
    physicsBodyDefinitions [{
      boundBone {boneName=dyn_wire_02_r}
      boxExtents=<40 40 30>
      constraintSetup {angularLimitsMax=<0 20 0>, angularLimitsMin=<0 -20 -20>, angularTarget=<1 0 0>, linearAxesMax=<1 0 1>, linearAxesMin=<-0.25 -0.25 0>}
      localJointOffset=<4 0 0>
     }]
    retargetingSettings {bFlipEasing=true, easingWeight=4}
   }
   n23 = Anim Dynamics - Bone: dyn_wire_01_l(ComponentPose=Local To Component(LocalPose=Use cached pose 'offset_CP'()), Alpha=1) {
    angularDampingOverride=1
    angularSpringConstant=20000
    bAngularSpring=true
    bLinearSpring=true
    bOverrideAngularDamping=true
    bOverrideLinearDamping=true
    bUseGravityOverride=true
    gravityScale=0.35
    linearDampingOverride=1
    linearSpringConstant=100
    simulationSpace=RootRelative
    alphaBoolBlend {blendOption=Linear}
    boundBone {boneName=dyn_wire_01_l}
    componentLinearAccScale=<1 1 1>
    componentLinearVelScale=<1 1 1>
    gravityOverride {z=-980}
    physicsBodyDefinitions [{
      boundBone {boneName=dyn_wire_01_l}
      boxExtents=<40 40 30>
      constraintSetup {angularLimitsMax=<0 20 0>, angularLimitsMin=<0 -20 -20>, angularTarget=<1 0 0>, linearAxesMax=<1 0 1>, linearAxesMin=<-0.25 -0.25 0>}
      localJointOffset=<-4 0 0>
     }]
    retargetingSettings {bFlipEasing=true, easingWeight=4}
   }
   n24 = Anim Dynamics - Bone: dyn_wire_02_l(ComponentPose=@n23, Alpha=1) {
    angularDampingOverride=1
    angularSpringConstant=20000
    bAngularSpring=true
    bLinearSpring=true
    bOverrideAngularDamping=true
    bOverrideLinearDamping=true
    bUseGravityOverride=true
    gravityScale=0.35
    linearDampingOverride=1
    linearSpringConstant=100
    simulationSpace=RootRelative
    alphaBoolBlend {blendOption=Linear}
    boundBone {boneName=dyn_wire_02_l}
    componentLinearAccScale=<1 1 1>
    componentLinearVelScale=<1 1 1>
    gravityOverride {z=-980}
    physicsBodyDefinitions [{
      boundBone {boneName=dyn_wire_02_l}
      boxExtents=<40 40 30>
      constraintSetup {angularLimitsMax=<0 20 0>, angularLimitsMin=<0 -20 -20>, angularTarget=<1 0 0>, linearAxesMax=<1 0 1>, linearAxesMin=<-0.25 -0.25 0>}
      localJointOffset=<-4 0 0>
     }]
    retargetingSettings {bFlipEasing=true, easingWeight=4}
   }
   n25 = Anim Dynamics - Bone: dyn_wire_01_r(ComponentPose=@n24, Alpha=1) {
    angularDampingOverride=1
    angularSpringConstant=20000
    bAngularSpring=true
    bLinearSpring=true
    bOverrideAngularDamping=true
    bOverrideLinearDamping=true
    bUseGravityOverride=true
    gravityScale=0.35
    linearDampingOverride=1
    linearSpringConstant=100
    simulationSpace=RootRelative
    alphaBoolBlend {blendOption=Linear}
    boundBone {boneName=dyn_wire_01_r}
    componentLinearAccScale=<1 1 1>
    componentLinearVelScale=<1 1 1>
    gravityOverride {z=-980}
    physicsBodyDefinitions [{
      boundBone {boneName=dyn_wire_01_r}
      boxExtents=<40 40 30>
      constraintSetup {angularLimitsMax=<0 20 0>, angularLimitsMin=<0 -20 -20>, angularTarget=<1 0 0>, linearAxesMax=<1 0 1>, linearAxesMin=<-0.25 -0.25 0>}
      localJointOffset=<4 0 0>
     }]
    retargetingSettings {bFlipEasing=true, easingWeight=4}
   }
   n26 = Anim Dynamics - Bone: dyn_wire_02_r(ComponentPose=@n25, Alpha=1) {
    angularDampingOverride=1
    angularSpringConstant=20000
    bAngularSpring=true
    bLinearSpring=true
    bOverrideAngularDamping=true
    bOverrideLinearDamping=true
    bUseGravityOverride=true
    gravityScale=0.35
    linearDampingOverride=1
    linearSpringConstant=100
    simulationSpace=RootRelative
    alphaBoolBlend {blendOption=Linear}
    boundBone {boneName=dyn_wire_02_r}
    componentLinearAccScale=<1 1 1>
    componentLinearVelScale=<1 1 1>
    gravityOverride {z=-980}
    physicsBodyDefinitions [{
      boundBone {boneName=dyn_wire_02_r}
      boxExtents=<40 40 30>
      constraintSetup {angularLimitsMax=<0 20 0>, angularLimitsMin=<0 -20 -20>, angularTarget=<1 0 0>, linearAxesMax=<1 0 1>, linearAxesMin=<-0.25 -0.25 0>}
      localJointOffset=<4 0 0>
     }]
    retargetingSettings {bFlipEasing=true, easingWeight=4}
   }
   Save cached pose 'Anim Dynamics'(Pose=Blend Poses by bool(BlendPose_0=Component To Local(ComponentPose=@n22), BlendPose_1=Component To Local(ComponentPose=@n26)) {blendType=Linear}) {cacheName=Anim Dynamics}
  ]}
 EventGraph {nodes [
   Event BlueprintUpdateAnimation()
   > n1 = "Gravity Override State Function (Parameter Asset) (Nativized)"(
    CharacterPartAnimInstance=Self-Reference()
    GameTimeInSeconds=GetGameTimeInSeconds()
    Parameters=&/F_MED_ShadeToll_Pack_Parameters
   )
   > Set GravityOverride(GravityOverride=@n1)
   > Set RBAN Settings Out(RBAN Settings Out=@n1.OutRigidBodyAnimNodeParameters)
   > Set is M_Large(is M_Large:bool=AND Boolean(A=Get isMale() -> (isMale:bool), B=Get isLarge() -> (isLarge:bool))) -> (Output_Get:bool)
   n2 = Event OnCharacterCustomizationLoaded()
   n3 = Get Array Item with EFort Custom Part Type(Array:Array of Custom Character Part Object References=@n2.CustomParts) -> (Output:Custom Character Part Object Reference)
   n4 = Get GenderPermitted!(self:Custom Character Part Object Reference=@n3) -> (GenderPermitted:EFortCustomGender Enum=Invalid)
   > Set isMale(isMale:bool="Equal (Enum)"(A:EFortCustomGender Enum=@n4, B:EFortCustomGender Enum=Male)) -> (Output_Get:bool)
   > Set isFemale(isFemale:bool="Equal (Enum)"(A=@n4, B=Female)) -> (Output_Get:bool)
   > Set isLarge(isLarge:bool="Equal (Enum)"(
     A:EFortCustomBodyType Enum=Get BodyTypesPermitted!(self:Custom Character Part Object Reference=@n3) -> (BodyTypesPermitted:EFortCustomBodyType Enum=NONE)
     B:EFortCustomBodyType Enum=Large
    )) -> (Output_Get:bool)
  ]}
]
variables [
 GravityOverride {type=Vector}
 is M_Large {type=boolean, default=true}
 is M_Large_0 {type=boolean, default=true}
 isFemale {type=boolean}
 isLarge {type=boolean}
 isMale {type=boolean}
 RBAN Settings Out {type=FortRigidBodyAnimNodeParameters}
]
